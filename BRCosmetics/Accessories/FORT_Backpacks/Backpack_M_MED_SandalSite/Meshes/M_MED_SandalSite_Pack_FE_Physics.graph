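# Physics asset for the SandalSite backpack cosmetic, providing bodies for its dangling sports gear, belt and zipper.
# Capsule-only bodies branch from the bag base through sticks, belt and zipper base bones, whose joints rotate freely.
# Two bats, a hockey stick and a golf club get tight swing limits, the latter two allowing more twist and using paired capsules.
# The belt forms a two-link limited chain, while the zipper swings mainly on one axis.
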PhysicsAsset /BRCosmetics/Accessories/FORT_Backpacks/Backpack_M_MED_SandalSite/Meshes/M_MED_SandalSite_Pack_FE_Physics
bodies {
 dyn_bag_base [{shapeType=Capsule, length=10, radius=8, shapeName=dyn_bag_base_capsule, center=<-10.9262 0 0>, rotation=<rot 90 180 -180>}]
 dyn_baseball_bat_1 [{
   shapeType=Capsule
   length=22.3499
   radius=2.7965
   shapeName=dyn_baseball_bat_1_capsule
   center=<29.5614 -0.0952 0.0218>
   rotation=<rot 0 -89.2605 -89.0329>
  }]
 dyn_baseball_bat_2 [{
   shapeType=Capsule
   length=23.538
   radius=1.4
   shapeName=dyn_baseball_bat_2_capsule
   center=<25.0359 0.8565 -0.0099>
   rotation=<rot 0 90.3475 -89.8665>
  }]
 dyn_belt1_02 [{shapeType=Capsule, length=12, radius=2.7, shapeName=dyn_belt1_02_capsule, center=<1.4909 -0.1061 1.4687>, rotation=<rot 0 -10 -84>}]
 dyn_belt_01 [{
   shapeType=Capsule
   length=11.9687
   radius=2.7434
   shapeName=dyn_belt_01_capsule
   center=<13.1537 -1.4539 2.6663>
   rotation=<rot 0 -10.766 -84.9831>
  }]
 dyn_belt_base [{shapeType=Capsule, length=0.505, radius=0.505, shapeName=dyn_belt_base_capsule, rotation=<rot 0 -90 0>}]
 dyn_golf_club [
  {shapeType=Capsule, length=25, radius=0.5, shapeName=dyn_golf_club_capsule, center=<23.686 -0.4809 0>, rotation=<rot 90 0 0>}
  {
   shapeType=Capsule
   length=2.5338
   radius=3.1137
   shapeName=dyn_golf_club_capsule
   center=<38.8232 -4.4358 -0.5434>
   rotation=<rot 0.0996 14.5493 89.5942>
  }
 ]
 dyn_hockey_stick [
  {
   shapeType=Capsule
   length=35.4458
   radius=0.5
   shapeName=dyn_hockey_stick_capsule
   center=<26.6194 -1.1319 0>
   rotation=<rot 90 160.5288 160.5288>
  }
  {shapeType=Capsule, length=35.9695, radius=0.5, shapeName=dyn_hockey_stick_capsule, center=<25.5255 0.9604 0>, rotation=<rot 90 0 0>}
 ]
 dyn_sticks_base [{shapeType=Capsule, length=0.505, radius=0.505, shapeName=dyn_sticks_base_capsule, rotation=<rot 0 -90 0>}]
 dyn_zipper [{shapeType=Capsule, length=0, radius=0.5, shapeName=dyn_zipper_capsule, center=<1.7023 0 0>}]
 dyn_zipper_base [{shapeType=Capsule, length=0, radius=1, shapeName=dyn_zipper_base_capsule}]
}
constraints [
 {bone1Name=dyn_zipper, bone2Name=dyn_zipper_base, swing1LimitDegrees=65, swing2LimitDegrees=5, twistLimitDegrees=2}
 {
  bone1Name=dyn_zipper_base
  bone2Name=dyn_bag_base
  swing1LimitDegrees=45
  swing1Motion=Free
  swing2LimitDegrees=45
  swing2Motion=Free
  twistLimitDegrees=45
  twistMotion=Free
 }
 {bone1Name=dyn_hockey_stick, bone2Name=dyn_sticks_base, swing1LimitDegrees=4, swing2LimitDegrees=4, twistLimitDegrees=40}
 {bone1Name=dyn_baseball_bat_1, bone2Name=dyn_sticks_base, swing1LimitDegrees=4, swing2LimitDegrees=5, twistLimitDegrees=4}
 {bone1Name=dyn_baseball_bat_2, bone2Name=dyn_sticks_base, swing1LimitDegrees=3, swing2LimitDegrees=5, twistLimitDegrees=2}
 {bone1Name=dyn_golf_club, bone2Name=dyn_sticks_base, swing1LimitDegrees=5, swing2LimitDegrees=6, twistLimitDegrees=30}
 {
  bone1Name=dyn_sticks_base
  bone2Name=dyn_bag_base
  swing1LimitDegrees=45
  swing1Motion=Free
  swing2LimitDegrees=45
  swing2Motion=Free
  twistLimitDegrees=45
  twistMotion=Free
 }
 {
  bone1Name=dyn_belt_base
  bone2Name=dyn_bag_base
  swing1LimitDegrees=45
  swing1Motion=Free
  swing2LimitDegrees=45
  swing2Motion=Free
  twistLimitDegrees=45
  twistMotion=Free
 }
 {bone1Name=dyn_belt_01, bone2Name=dyn_belt_base, swing1LimitDegrees=14, swing2LimitDegrees=40, twistLimitDegrees=10}
 {bone1Name=dyn_belt1_02, bone2Name=dyn_belt_01, swing1LimitDegrees=7, swing2LimitDegrees=5, twistLimitDegrees=2}
]
